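; GrillMoon POI ambience component that manages cosmetic-only audio for the point of interest's ambience, portal and trail, plus a post-event variant.
; Custom events spawn stored location-based sounds and a volume that ducks non-GrillMoon ambient loops, then fade the sounds out on stop.
; Audio parameters set trail volume from a map keyed by trail count and trigger portal ducking.
; A separate event toggles a control bus mix.
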(event Custom|Spawn%20POI%20Ambience
  (Networking|CanExecuteCosmeticEvents
    (:True
      (bind _returnvalue (Game|SpawnActorfromClass "/GrillMoon_Audio/Gameplay/BP_GrillMoon_POI_Volume.BP_GrillMoon_POI_Volume_C" (struct :Location "-57744.000000,42687.000000,9864.000000" :Rotation "0, 0, 0" :Scale "1.000000,1.000000,1.000000") "Undefined" "MultiplyWithRoot" 0 0)) ; Ducks non-GrillMoon ambient loops so they don't clash with GrillMoon POI ambient loops
      (Variables|Default|SetPOI%20Volume _returnvalue)
      (Development|PrintString "Spawned GrillMoon POI Volume")
      (Utilities|FlowControl|Sequence
        (:then_0
          (Variables|Default|GetAmbience%20Audio
            (:then
              (Development|PrintString "Attempted to spawn GrillMoon POI ambience but there was already an active audio component."))
            (:else
              (bind _returnvalue_2 (Audio|SpawnSoundatLocation (Variables|Default|GetAmbience%20Sound) (Variables|Default|GetLocation)))
              (Variables|Default|SetAmbience%20Audio _returnvalue_2)
              (bind _returnvalue_3 (Audio|SpawnSoundatLocation (Variables|Default|GetPortal%20Sound) (Variables|Default|GetPortal%20Location)))
              (Variables|Default|SetPortal%20Audio _returnvalue_3))))
        (:then_1
          (Utilities|FlowControl|Sequence
            (:then_0
              (bind _returnvalue_4 (Audio|SpawnSoundatLocation (Variables|Default|GetTrail%20Sound) (Variables|Default|GetTrail%20Location)))
              (bind _output_get (Variables|Default|SetTrail%20Audio _returnvalue_4))
              (Audio|Parameter|SetFloatParameter _output_get "TrailVolume" (Utilities|Map|Find (Variables|Default|GetTrail%20Volumes) (Variables|Default|GetNumber%20of%20Trails)) :!class /Script/AudioExtensions.AudioParameterControllerInterface))))))))

(event Custom|Increment%20Trail%20Volume
  (Networking|CanExecuteCosmeticEvents
    (:True
      (bind _trail_audio (Variables|Default|GetTrail%20Audio)
        (:then
          (Audio|Parameter|SetFloatParameter _trail_audio "TrailVolume" (Utilities|Map|Find (Variables|Default|GetTrail%20Volumes) (Math|Integer|Clamp(Integer) (+ (Variables|Default|GetNumber%20of%20Trails) 1) 0 5)) :!class /Script/AudioExtensions.AudioParameterControllerInterface))
        (:else)))))

(event Custom|Stop%20POI%20Ambience
  (Networking|CanExecuteCosmeticEvents
    (:True
      (Utilities|FlowControl|Sequence
        (:then_0
          (bind _ambience_audio (Variables|Default|GetAmbience%20Audio)
            (:then
              (Audio|Components|Audio|FadeOut _ambience_audio 2.0 0.0 "Linear" :!class /Script/Engine.AudioComponent))
            (:else)))
        (:then_1
          (Utilities|FlowControl|Sequence
            (:then_0
              (bind _trail_audio (Variables|Default|GetTrail%20Audio)
                (:then
                  (Audio|Components|Audio|FadeOut _trail_audio 2.0 0.0 "Linear" :!class /Script/Engine.AudioComponent))
                (:else)))
            (:then_1
              (Utilities|FlowControl|Sequence
                (:then_0
                  (bind _portal_audio (Variables|Default|GetPortal%20Audio)
                    (:then
                      (Audio|Components|Audio|FadeOut _portal_audio 2.0 0.0 "Linear" :!class /Script/Engine.AudioComponent))
                    (:else)))))))))))

(event Custom|Set%20Ambience%20Control%20Bus%20Mix%20Activation
  (params (Activate :type Boolean))
  (bind _music_control_bus_mix (Variables|Default|GetMusic%20Control%20Bus%20Mix))
  (Networking|CanExecuteCosmeticEvents
    (:True
      (if Activate
        (Audio|Modulation|ActivateControlBusMix _music_control_bus_mix)
        (Development|PrintString "Activated CBM_GrillMoon_Ambience" true true "(R=1.000000,G=0.000000,B=0.258183,A=1.000000)")
        (else
          (Audio|Modulation|DeactivateControlBusMix _music_control_bus_mix)
          (Development|PrintString "Deactivated CBM_GrillMoon_Ambience" true true "(R=1.000000,G=0.000000,B=0.258183,A=1.000000)"))))))

(event Custom|Spawn%20Post%20Event%20Ambience
  (bind _post_event_portal_sound (Variables|Default|GetPost%20Event%20Portal%20Sound))
  (Networking|CanExecuteCosmeticEvents
    (:True
      (Utilities|FlowControl|Sequence
        (:then_0
          (Variables|Default|GetAmbience%20Audio
            (:then
              (Development|PrintString "Attempted to spawn post event GrillMoon POI ambience but there was already an active audio component."))
            (:else
              (bind _returnvalue (Audio|SpawnSoundatLocation (Variables|Default|GetPost%20Event%20Ambience) (Variables|Default|GetLocation)))
              (Variables|Default|SetAmbience%20Audio _returnvalue))))
        (:then_1
          (Utilities|FlowControl|Sequence
            (:then_0
              (bind _portal_audio (Variables|Default|GetPortal%20Audio)
                (:then
                  (Audio|Components|Audio|SetSound _portal_audio _post_event_portal_sound)
                  (Audio|Components|Audio|FadeIn _portal_audio 1.0 1.0 0.0 "Linear" :!class /Script/Engine.AudioComponent))
                (:else
                  (bind _returnvalue_2 (Audio|SpawnSoundatLocation _post_event_portal_sound (Variables|Default|GetPortal%20Location)))
                  (Variables|Default|SetPortal%20Audio _returnvalue_2))))))))))

(event Custom|Duck%20Portal%20Sound
  (Networking|CanExecuteCosmeticEvents
    (:True
      (bind _portal_audio (Variables|Default|GetPortal%20Audio)
        (:then
          (Audio|Parameter|ExecuteTriggerParameter _portal_audio "DuckPortalSound"))
        (:else)))))
