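; GrillMoon audio component that starts and stops point-of-interest music, running playback only where cosmetic events can execute.
; Start spawns the sound at a stored location, unless an audio component already exists, and registers it with the MusicEventSubsystem as a POI event.
; Stop fades it out and unregisters it when audio finishes; EndPlay also unregisters.
; A separate event toggles the music control bus mix, with debug prints.
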(event Custom|Start%20POI%20Music
  (Networking|CanExecuteCosmeticEvents
    (:True
      (Variables|Default|GetMusic%20Audio
        (:then
          (Development|PrintString "Attempted to play GrillMoon POI music but there was already an active audio component."))
        (:else
          (bind _returnvalue (Audio|SpawnSoundatLocation (Variables|Default|GetMusic) (Variables|Default|GetLocation)))
          (Variables|Default|SetMusic%20Audio _returnvalue)
          (Audio|AddEventbyComponent (GameInstanceSubsystems|GetMusicEventSubsystem) (Variables|Default|GetMusic%20Audio) "(TagName=\"MusicEventSubsystem.Event.POI\")" "(GameplayTags=)" 0 ""))))))

(event Custom|Stop%20POI%20Music
  (Networking|CanExecuteCosmeticEvents
    (:True
      (bind _music_audio (Variables|Default|GetMusic%20Audio)
        (:then
          (Audio|Components|Audio|FadeOut _music_audio (Variables|Default|GetPOI%20Music%20Fade%20Out%20Duration) 0.0 "Linear" :!class /Script/Engine.AudioComponent)
          (EventDispatchers|BindEventtoOnAudioFinished _music_audio (delegate Custom|Unregister)))
        (:else)))))

(event Custom|Unregister
  (Audio|RemoveEvent (GameInstanceSubsystems|GetMusicEventSubsystem) (Variables|Default|GetMusic%20Audio)))

(event EventEndPlay
  (params (EndPlayReason :type EEndPlayReason%20Enum :default "Destroyed"))
  (Audio|RemoveEvent (GameInstanceSubsystems|GetMusicEventSubsystem) (Variables|Default|GetMusic%20Audio)))

(event Custom|Set%20Music%20Control%20Bus%20Mix%20Activation
  (params (Activate :type Boolean))
  (bind _music_control_bus_mix (Variables|Default|GetMusic%20Control%20Bus%20Mix))
  (Networking|CanExecuteCosmeticEvents
    (:True
      (if Activate
        (Audio|Modulation|ActivateControlBusMix _music_control_bus_mix)
        (Development|PrintString "Activated CBM_GrillMoon_Music" true true "(R=1.000000,G=0.000000,B=0.258183,A=1.000000)")
        (else
          (Audio|Modulation|DeactivateControlBusMix _music_control_bus_mix)
          (Development|PrintString "Deactivated CBM_GrillMoon_Music" true true "(R=1.000000,G=0.000000,B=0.258183,A=1.000000)"))))))
