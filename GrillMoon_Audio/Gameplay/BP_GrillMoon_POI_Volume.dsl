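; GrillMoon POI volume Blueprint that changes the audio mix while the locally viewed player pawn is inside it.
; On begin overlap, it casts the other actor to PlayerPawn_Athena and, if locally viewed, pushes the Sound Class Mix variable as a sound mix modifier.
; End overlap pops the same mix under the same check.
; Both paths run only where cosmetic events can execute, and print debug strings.
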(event Collision|EventActorBeginOverlap
  (params (OtherActor :type Actor%20Object%20Reference))
  (Networking|CanExecuteCosmeticEvents
    (:True
      (bind _asplayer_pawn_athena (Utilities|Casting|CastToPlayerPawn_Athena OtherActor))
      (if (Pawn|IsLocallyViewed _asplayer_pawn_athena)
        (Audio|PushSoundMixModifier (Variables|Default|GetSound%20Class%20Mix))
        (Development|PrintString "Activated GrillMoon_POI_MixMod")))))

(event Collision|EventActorEndOverlap
  (params (OtherActor :type Actor%20Object%20Reference))
  (Networking|CanExecuteCosmeticEvents
    (:True
      (bind _asplayer_pawn_athena (Utilities|Casting|CastToPlayerPawn_Athena OtherActor))
      (if (Pawn|IsLocallyViewed _asplayer_pawn_athena)
        (Audio|PopSoundMixModifier (Variables|Default|GetSound%20Class%20Mix))
        (Development|PrintString "Deactivated GrillMoon_POI_MixMod")))))

(fn ConstructionScript)
